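; Gameplay ability for the VividRazor Super Shield medallion that destroys the spawned shield when the ability ends, presumably on medallion drop as the name suggests.
; On activation it waits for the ShieldSpawned gameplay event, casts the payload's OptionalObject to BGA_Athena_SilverBlazerCore_Mini and stores it as SpawnedShield.
; On ability end, cancelled or not, it gives a valid stored shield a very short lifespan rather than destroying it directly.
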(event Ability|EventActivateAbility
  (bind (_asynctaskproxy _payload) (Ability|Tasks|WaitGameplayEvent "(TagName=\"Event.Athena.VividRazorMedallions.SuperShield.ShieldSpawned\")")
    (:EventReceived
      (bind _asbga_athena_silver_blazer_core_mini (Utilities|Casting|CastToBGA_Athena_SilverBlazerCore_Mini (.OptionalObject (Utilities|Struct|BreakGameplayEventData _payload))))
      (Variables|Default|SetSpawnedShield _asbga_athena_silver_blazer_core_mini))))

(event Ability|EventOnEndAbility
  (params (bWasCancelled :type Boolean))
  (bind _spawnedshield (Variables|Default|GetSpawnedShield)
    (:then
      (Actor|SetLifeSpan _spawnedshield 0.1))
    (:else)))
